# Copyright Epic Games, Inc. All Rights Reserved.
#################################################
# Generated Digest of Verse API
# DO NOT modify this manually!
# Generated from build: ++Fortnite+Release-36.20-CL-43847582
#################################################


using {/Verse.org/Native}
using {/Verse.org/Simulation}
using {/Verse.org/SceneGraph}
using {/Verse.org/Assets}
# Module import path: /playkigai.dig@fortnite.com/Minigame_BombRun/Materials
Materials := module:
    M_Checkpoint<scoped {/playkigai.dig@fortnite.com/Minigame_BombRun}> := class<final><scoped {/playkigai.dig@fortnite.com/Minigame_BombRun}>(material):


    M_DeathVolume<scoped {/playkigai.dig@fortnite.com/Minigame_BombRun}> := class<final><scoped {/playkigai.dig@fortnite.com/Minigame_BombRun}>(material):


# Module import path: /playkigai.dig@fortnite.com/Minigame_BombRun/Meshes
Meshes := module:
    SM_Arrow<scoped {/playkigai.dig@fortnite.com/Minigame_BombRun}> := class<final><scoped {/playkigai.dig@fortnite.com/Minigame_BombRun}>(mesh_component):
        @editable
        var Material_0<public>:material = external {}

    SM_Arrow_asset<scoped {/playkigai.dig@fortnite.com/Minigame_BombRun}>:mesh = external {}

    SM_Bomb<scoped {/playkigai.dig@fortnite.com/Minigame_BombRun}> := class<final><scoped {/playkigai.dig@fortnite.com/Minigame_BombRun}>(mesh_component):
        @editable
        var WorldGridMaterial<public>:material = external {}

    SM_Bomb_asset<scoped {/playkigai.dig@fortnite.com/Minigame_BombRun}>:mesh = external {}

    SM_CharacterCollision<scoped {/playkigai.dig@fortnite.com/Minigame_BombRun}> := class<final><scoped {/playkigai.dig@fortnite.com/Minigame_BombRun}>(mesh_component):
        @editable
        var Material_0<public>:material = external {}

    SM_CharacterCollision_asset<scoped {/playkigai.dig@fortnite.com/Minigame_BombRun}>:mesh = external {}

    SM_Checkpoint<scoped {/playkigai.dig@fortnite.com/Minigame_BombRun}> := class<final><scoped {/playkigai.dig@fortnite.com/Minigame_BombRun}>(mesh_component):
        @editable
        var WorldGridMaterial<public>:material = external {}

    SM_Checkpoint_asset<scoped {/playkigai.dig@fortnite.com/Minigame_BombRun}>:mesh = external {}

    SM_DeathVolume<scoped {/playkigai.dig@fortnite.com/Minigame_BombRun}> := class<final><scoped {/playkigai.dig@fortnite.com/Minigame_BombRun}>(mesh_component):
        @editable
        var Material_0<public>:material = external {}

    SM_DeathVolume_asset<scoped {/playkigai.dig@fortnite.com/Minigame_BombRun}>:mesh = external {}

    SM_StartingPlate<scoped {/playkigai.dig@fortnite.com/Minigame_BombRun}> := class<final><scoped {/playkigai.dig@fortnite.com/Minigame_BombRun}>(mesh_component):


    SM_StartingPlate_asset<scoped {/playkigai.dig@fortnite.com/Minigame_BombRun}>:mesh = external {}

# Module import path: /playkigai.dig@fortnite.com/Minigame_BombRun/Prefabs
Prefabs := module:
    @import_as("/Minigame_BombRun/Prefabs/bomb.bomb_C")
    bomb<scoped {/playkigai.dig@fortnite.com/Minigame_BombRun}> := class<final>(entity):


    bomb_asset<scoped {/playkigai.dig@fortnite.com/Minigame_BombRun}>:entity_prefab = external {}

    @import_as("/Minigame_BombRun/Prefabs/character_collision_proxy.character_collision_proxy_C")
    character_collision_proxy<scoped {/playkigai.dig@fortnite.com/Minigame_BombRun}> := class<final>(entity):


    character_collision_proxy_asset<scoped {/playkigai.dig@fortnite.com/Minigame_BombRun}>:entity_prefab = external {}

    @import_as("/Minigame_BombRun/Prefabs/checkpoint.checkpoint_C")
    checkpoint<scoped {/playkigai.dig@fortnite.com/Minigame_BombRun}> := class<final>(entity):


    checkpoint_asset<scoped {/playkigai.dig@fortnite.com/Minigame_BombRun}>:entity_prefab = external {}

    @import_as("/Minigame_BombRun/Prefabs/collision_proxy_spawner.collision_proxy_spawner_C")
    collision_proxy_spawner<scoped {/playkigai.dig@fortnite.com/Minigame_BombRun}> := class<final>(entity):


    collision_proxy_spawner_asset<scoped {/playkigai.dig@fortnite.com/Minigame_BombRun}>:entity_prefab = external {}

    @import_as("/Minigame_BombRun/Prefabs/death_volume.death_volume_C")
    death_volume<scoped {/playkigai.dig@fortnite.com/Minigame_BombRun}> := class<final>(entity):


    death_volume_asset<scoped {/playkigai.dig@fortnite.com/Minigame_BombRun}>:entity_prefab = external {}

    @import_as("/Minigame_BombRun/Prefabs/jump_pad.jump_pad_C")
    jump_pad<scoped {/playkigai.dig@fortnite.com/Minigame_BombRun}> := class<final>(entity):


    jump_pad_asset<scoped {/playkigai.dig@fortnite.com/Minigame_BombRun}>:entity_prefab = external {}

    @import_as("/Minigame_BombRun/Prefabs/race_manager.race_manager_C")
    race_manager<scoped {/playkigai.dig@fortnite.com/Minigame_BombRun}> := class<final>(entity):


    race_manager_asset<scoped {/playkigai.dig@fortnite.com/Minigame_BombRun}>:entity_prefab = external {}

    @import_as("/Minigame_BombRun/Prefabs/start_race.start_race_C")
    start_race<scoped {/playkigai.dig@fortnite.com/Minigame_BombRun}> := class<final>(entity):


    start_race_asset<scoped {/playkigai.dig@fortnite.com/Minigame_BombRun}>:entity_prefab = external {}

    @import_as("/Minigame_BombRun/Prefabs/target_arrow.target_arrow_C")
    target_arrow<scoped {/playkigai.dig@fortnite.com/Minigame_BombRun}> := class<final>(entity):


    target_arrow_asset<scoped {/playkigai.dig@fortnite.com/Minigame_BombRun}>:entity_prefab = external {}

    @import_as("/Minigame_BombRun/Prefabs/ui_manager.ui_manager_C")
    ui_manager<scoped {/playkigai.dig@fortnite.com/Minigame_BombRun}> := class<final>(entity):


    ui_manager_asset<scoped {/playkigai.dig@fortnite.com/Minigame_BombRun}>:entity_prefab = external {}