<#
@UUID: 06223632-9beb-4da6-a575-355a325662a6
@Task: Make a component that will contain logic for the timers and bombs.
@Description: Implement a SceneGraph component that will manage the timers (logic and UI) and bombs.
@Tags: Timer, Bombs, UI
@Export: true
#>

using { /Fortnite.com/Characters }
using { /UnrealEngine.com }
using { /Verse.org }
using { /Verse.org/SceneGraph }
using { /Verse.org/Simulation }
using { /Verse.org/SpatialMath }

# Class for the logic of the death_volume entity.
bomb_component<public> := class<final_super>(component):
    @editable
    var PlayerTimer<public>: int = 15

    @editable
    var AddTimerValue<public>: int = 5

    @editable
    var BombDamage<public>: float = 200.0

    var MaybeAgent<public>: ?agent = false
    var PostTickHandler: ?cancelable = false
    var IsPlayerAlive: logic = false

    Offset: float = 200.0


    # Init bomb component.
    Init<public>(InAgent:agent):void =
        set MaybeAgent = option. InAgent
        set IsPlayerAlive = true

        if:
            Agent := MaybeAgent?
        then:
            if(Fort := Agent.GetFortCharacter[]):
                spawn. AttachBombTo(Fort)
                spawn. DecreaseTimers(Fort)


    # Decrease timer for InFort.
    DecreaseTimers(InFort:fort_character)<suspends>:void=
        if(IsPlayerAlive?):
            if (PlayerTimer > 0):
                set PlayerTimer -= 1
            else:
                InFort.Damage(BombDamage)
                set IsPlayerAlive = false

                # Call function to declare the winner.
                if:
                    Sim := Entity.GetSimulationEntity[]
                    RaceManagers := Sim.FindDescendantEntitiesWithComponent(race_manager_component)
                then:
                    # Search in the Race Manager the other player alive to set them as winner.
                    for (RaceManager : RaceManagers, Comp := RaceManager.GetComponent[race_manager_component]):
                        if (RedPlayer := Comp.MaybeRedPlayer?, BluePlayer := Comp.MaybeBluePlayer?):
                            if (RedPlayer = MaybeAgent?):
                                Comp.DeclareWinner(BluePlayer)
                            else:
                                Comp.DeclareWinner(RedPlayer)

            Sleep(1.0)
            DecreaseTimers(InFort)
    

    # Increase timer of the agent.
    IncreaseTimer<public>(Agent:agent):void=
        set PlayerTimer = PlayerTimer + AddTimerValue


    # Remove bomb from the player and set logic as false.
    RemoveBombsFrom<public>():void=
        set IsPlayerAlive = false
        Entity.RemoveFromParent()


    # Move the bomb with the player.
    AttachBombTo(Fort:fort_character)<suspends>:void=
        if (IsPlayerAlive?):
            loop:
                Sleep(0.001)
                CharacterTranslation := Fort.GetTransform().Translation
                    Entity.SetGlobalTransform(transform{ 
                        Translation := vector3{
                            Forward := CharacterTranslation.X,
                            Left := -CharacterTranslation.Y,
                            Up := CharacterTranslation.Z + Offset
                        }, 
                        Rotation := Entity.GetGlobalTransform().Rotation, 
                        Scale := Entity.GetGlobalTransform().Scale
                    })


#Extend the bomb component with a GetAgent() function.
(Bomb:Prefabs.bomb).GetAgent<public>()<decides><transacts>:agent=
    BombComponent := Bomb.GetComponent[bomb_component]
    BombComponent.MaybeAgent?