<#
@UUID: 5806a391-69d2-45e0-b2cc-c3f260b8b25a
@Task: Make a component to detect collisions between entities and agents.
@Description: Implement a SceneGraph component that detects collisions between the agent it is attached to and other entities.
@Tags: CollisionDetection, AgentCollisionDetection
@Export: true
#>

using { /Fortnite.com/Characters }
using { /UnrealEngine.com }
using { /Verse.org }
using { /Verse.org/Native }
using { /Verse.org/Random }
using { /Verse.org/Simulation }
using { /Verse.org/SpatialMath }
using { /Verse.org/SceneGraph }
using { /Verse.org/SceneGraph/KeyframedMovement }
using { Prefabs }

# A component that acts as a proxy for the FortCharacter collision, since entities do not detect collision with anything other than entities.
character_collision_proxy_component<public> := class<final_super>(component):
    var MaybeAgent<public>: ?agent = false
    var PostTickHandler: ?cancelable = false


    Init<public>(InAgent:agent):void =
        set MaybeAgent = option. InAgent


    OnBeginSimulation<override>():void =
        (super:)OnBeginSimulation()
        set PostTickHandler = option. TickEvents.PostPhysics.Subscribe(OnPostPhysics)


    # We set this character_collision_proxy_component's transform equal to the MaybeAgent's FortCharacter's Transform.
    OnPostPhysics<public>(Delta:float): void =
        if (FortCharacter := MaybeAgent?.GetFortCharacter[]):
            CharacterTranslation := FortCharacter.GetTransform().Translation
            Entity.SetGlobalTransform(transform{ 
                Translation := vector3{
                    Forward := CharacterTranslation.X,
                    Left := -CharacterTranslation.Y,
                    Up := CharacterTranslation.Z
                }, 
                Rotation := Entity.GetGlobalTransform().Rotation, 
                Scale := Entity.GetGlobalTransform().Scale
            })


# Extend the character_collision_proxy with a GetAgent() function.
(CollisionProxy:character_collision_proxy).GetAgent<public>()<decides><transacts>:agent=
    ProxyComponent := CollisionProxy.GetComponent[character_collision_proxy_component]
    ProxyComponent.MaybeAgent?