<#
@UUID: c70227c6-b258-4fd5-827a-cdef756a5ce5
@Task: Make a component to increase the timer on collision.
@Description: Implement a SceneGraph component that increases the timer, enables the next checkpoint, and disables the current checkpoint after collision.
@Tags: Collision Detection, Checkpoint
@Export: true
#>

using { /UnrealEngine.com }
using { /Verse.org }
using { /Verse.org/SceneGraph }
using { /Verse.org/Simulation }
using { /Verse.org/SpatialMath }

# === Custom Event Class ===

# Custom function type for CheckpointReachedSubscribable.
checkpoint_reached_function := type{_(:agent, :int):void}

# Class for custom subscribable.
checkpoint_reached_subscribable<internal> := class:
    var Functions<private>:[]checkpoint_reached_function = array{}

    # Subscribe-like function.
    # Needs to be wrapped by checkpoint_component.
    # Accepts functions of type checkpoint_reached_function, which will be run during Dispatch().
    AddFunction(InFunction:checkpoint_reached_function):void=
        set Functions += array{InFunction}

    # Call all functions registered.
    Dispatch(InAgent:agent, InIndex:int):void=
        for (Function : Functions):
            Function(InAgent,InIndex)

# === End Custom Event Class ===

checkpoint_component<public> := class<final_super>(component):
    var AgentToIsTouchingCheckpointMap<private>:[agent]logic = map{}
    var AgentToPlayerLightComponent:[agent]?light_component = map{}
    var NextCheckpointLocation<private>:?transform = false
    # References to set in the editor, so we know wich light corresponds to wich player
    @editable
    MaybeBluePlayerLigth:?entity = false
    @editable
    MaybeRedPlayerLigth:?entity = false
    @editable
    var CheckpointIndex : int = -1

    CheckpointSubscribable:checkpoint_reached_subscribable = checkpoint_reached_subscribable{}

    # === Entity Functions === 

    OnBeginSimulation<override>():void =
        (super:)OnBeginSimulation()
        if(Mesh := Entity.GetComponent[mesh_component]):
            Mesh.EntityEnteredEvent.Subscribe(OnEntityHit)

    # Handle agent collision.
    OnEntityHit<private>(HitEntity:entity):void=
        if:
            CharacterCollisionProxy := Prefabs.character_collision_proxy[HitEntity]
            Agent := CharacterCollisionProxy.GetAgent[] 
        then:
            CheckpointSubscribable.Dispatch(Agent,CheckpointIndex)

    # Handle agent leaving collision.
    OnEntityExitCollision<private>(HitEntity:entity):void=    
        if:
            CharacterCollisionProxy := Prefabs.character_collision_proxy[HitEntity]
            Agent := CharacterCollisionProxy.GetAgent[]
            AgentToIsTouchingCheckpointMap[Agent] = false

    # === End Entity Functions ===

    # === Checkpoint Functions ===

    # Enable checkpoint for specific player.
    EnableCheckpointForPlayer<public>(InAgent:agent, IsBluePlayer:logic):void=
        # Enable the player active light.
        var MaybePlayerLight:?entity = false
        if. IsBluePlayer? then. set MaybePlayerLight = MaybeBluePlayerLigth else. set MaybePlayerLight = MaybeRedPlayerLigth
        if:
            PlayerLight := MaybePlayerLight?
            LightComp := PlayerLight.GetComponent[spot_light_component]
        then:
            LightComp.Enable()

    # Disable checkpoint for specific player.
    DisableCheckpointForPlayer<public>(InAgent:agent, IsBluePlayer:logic):void=
        # Disable player active light.
        var MaybePlayerLight:?entity = false
        if. IsBluePlayer? then. set MaybePlayerLight = MaybeBluePlayerLigth else. set MaybePlayerLight = MaybeRedPlayerLigth
        if:
            PlayerLight := MaybePlayerLight?
            LightComp := PlayerLight.GetComponent[spot_light_component]
        then:
            LightComp.Disable()

    # Reset variables for new race.
    Reset<public>():void=
        if:
            BlueLigth := MaybeBluePlayerLigth?.GetComponent[spot_light_component]
            RedLigth := MaybeRedPlayerLigth?.GetComponent[spot_light_component]
        then:
            BlueLigth.Disable()
            RedLigth.Disable()
    SetIndex<public>(InInt:int):void=
        set CheckpointIndex = InInt
    # Subscribe to CheckpointReached custom event.
    Subscribe<public>(InFunction:checkpoint_reached_function):void=
        CheckpointSubscribable.AddFunction(InFunction)

    # === End Checkpoint Functions ===

    # === Helpers ===

    # Check if this checkpoint is inactive.
    # Succeeds if there are no player lights active, fails if there is at least one player light active.
    IsCheckpointInactive<private>()<decides><transacts>:void=
        for:
            PlayerLight : AgentToPlayerLightComponent
        do: 
            not PlayerLight?.IsEnabled[]

    
    # === End Helpers ===