<#
@UUID: 15faaf6a-7650-4305-834e-83962fba7847
@Task: A component that detects when the player goes off the track.
@Description: Implement a SceneGraph component that will teleport the player back to its last checkpoint upon colliding with it.
@Tags: Collision, Gameplay
@Export: true
#>

using { /Fortnite.com/Characters }
using { /Verse.org/SceneGraph }
using { /Verse.org/Simulation }
using { /UnrealEngine.com/Temporary/SpatialMath }
using { Prefabs }

# Class for the logic of the death_volume entity.
death_volume_component<public> := class<final_super>(component):

    var MaybeMesh: ?mesh_component = false


    OnBeginSimulation<override>():void=
        if (Mesh := Entity.GetComponent[mesh_component]):
            set MaybeMesh = option. Mesh
            Mesh.EntityEnteredEvent.Subscribe(OnEntityEntered)


    OnEntityEntered(EnteredEntity:entity):void=
        if:
            Proxy := character_collision_proxy[EnteredEntity] 
            Agent := Proxy.GetAgent[]
            Sim := Entity.GetSimulationEntity[]
            RaceManagers := Sim.FindDescendantEntitiesWithComponent(race_manager_component)
        then:
            TeleportPlayerToLastCheckpoint(Agent, RaceManagers)
    

    # Teleport player to last checkpoint.
    TeleportPlayerToLastCheckpoint(InAgent:agent, RaceManagers:generator(entity)):void=
        var HasBeenTeleported: logic = false

        if (Fort := InAgent.GetFortCharacter[]):
            for (RaceManager : RaceManagers, Comp := RaceManager.GetComponent[race_manager_component]):
                if:
                    MaybeIndex := Comp.AgentToCheckpointIndexMap[InAgent]? - 1
                    Checkpoint := Comp.RaceCheckpoints[MaybeIndex]
                    Transform := Checkpoint.GetGlobalTransform()
                    Fort.TeleportTo[(/UnrealEngine.com/Temporary/SpatialMath:)vector3 {
                            X := Transform.Translation.Forward,
                            Y := -Transform.Translation.Left,
                            Z := Transform.Translation.Up + 300.0
                        }, 
                        (/UnrealEngine.com/Temporary/SpatialMath:)IdentityRotation()
                    ]
                then:
                    set HasBeenTeleported = true