<#
@UUID: fa5b8782-dff7-4579-88a4-2e63586469f9
@Task: Make a component that will launch the player upwards when colliding with it.
@Description: Implement a SceneGraph component that will launch the player upwards when colliding with it.
@Tags: Physics, Movement
@Export: true
#>

using { /Fortnite.com/Devices }
using { /Verse.org/SceneGraph }
using { /Verse.org/Simulation }
using { /UnrealEngine.com/Temporary/SpatialMath }
using {Prefabs}

jump_pad_component<public> := class<final_super>(component):
    # Since we cannot directly apply forces to agents, we have to rely on a bouncer_device.
    @editable
    MaybeBouncerDevice: ?bouncer_device = false

    var JumpPadStartPos:(/UnrealEngine.com/Temporary/SpatialMath:)vector3 = (/UnrealEngine.com/Temporary/SpatialMath:)vector3{}

    OnBeginSimulation<override>():void=
        if(BouncerDevice := MaybeBouncerDevice?):
            set JumpPadStartPos = BouncerDevice.GetTransform().Translation
            if(Mesh := Entity.GetComponent[mesh_component]):
                Mesh.EntityEnteredEvent.Subscribe(OnEntityEntered)
            BouncerDevice.BouncedEvent.Subscribe(DeactivateJumpPad)


    OnEntityEntered(EnteredEntity:entity):void=
        if(Proxy := character_collision_proxy[EnteredEntity], Agent := Proxy.GetAgent[]):
            ActivateJumpPad()

    ActivateJumpPad():void=
        if(BouncerDevice := MaybeBouncerDevice?):
            BouncerDevice.Enable()
            if. BouncerDevice.TeleportTo[
                    (/UnrealEngine.com/Temporary/SpatialMath:)vector3
                        {
                            X := Entity.GetGlobalTransform().Translation.Forward, 
                            Y := -(Entity.GetGlobalTransform().Translation.Left + 50.0),
                            Z := Entity.GetGlobalTransform().Translation.Up
                        },
                        (/UnrealEngine.com/Temporary/SpatialMath:)IdentityRotation()]
    
    DeactivateJumpPad(MaybeAgent:?agent):void=
        if(BouncerDevice := MaybeBouncerDevice?):
            if. BouncerDevice.TeleportTo[JumpPadStartPos,(/UnrealEngine.com/Temporary/SpatialMath:)IdentityRotation()]
            BouncerDevice.Disable()