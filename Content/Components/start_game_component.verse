<#
@UUID: 94138db8-c7fe-453f-a554-b8a40ba19d0e
@Task: Make a component that will contain logic to start the race.
@Description: Implement a SceneGraph component that will manage the race start.
@Tags: RaceManagement
@Export: true
#>

using { /Fortnite.com/Characters }
using { /Fortnite.com/Playspaces }
using { /Fortnite.com/UI }
using { /UnrealEngine.com/Temporary/SpatialMath }
using { /UnrealEngine.com/Temporary/UI }
using { /Verse.org/Colors }
using { /Verse.org/SceneGraph }
using { /Verse.org/Simulation }
using { /Verse.org/SpatialMath }

start_game_component<public> := class<final_super>(component):

    @editable
    MaybeRedSpawnPosition: ?entity = false
    @editable
    MaybeBlueSpawnPosition: ?entity = false

    Countdown: float = 3.0
    var RaceStarted<public>: logic = false
    var AgentToBombs<private>: [agent]bomb_component = map{}

    var MaybeRedPlayer: ?agent = false
    var MaybeBluePlayer: ?agent = false
    var RedPlayerReady: logic = false
    var BluePlayerReady: logic = false
    var PlayersUI: [agent]timer_ui = map{}

    # Funcion to init UI adding it to each player.
    InitUI(Player:player):void=
        if: 
            set PlayersUI[Player] = timer_ui{}
            PlayerCanvas := PlayersUI[Player]
            PlayerUI := GetPlayerUI[Player]
        then:
            PlayerUI.AddWidget(PlayerCanvas.Canvas)

    # Funcion to remove UI from players.
    RemoveUI(Player:player):void=
        if:
            PlayerCanvas := PlayersUI[Player]
            PlayerUI := GetPlayerUI[Player]
        then. PlayerUI.RemoveWidget(PlayerCanvas.Canvas)


    OnBeginSimulation<override>():void=
        (super:)OnBeginSimulation()

        if (RedEntity := MaybeRedSpawnPosition?, BlueEntity := MaybeBlueSpawnPosition?):
            Print("Spawns correctly set up")
        else:
            Err("Spawns entity reference missing.")

        if (not RaceStarted?):
            # Check interaction with the starting buttons.
            Interactables := for(Interactable:Entity.FindDescendantEntitiesWithComponent(race_start_component), Comp := Interactable.GetComponent[race_start_component]). Comp
            if. Interactables[0].SucceededEvent.Subscribe(OnGameStartButtonRedInteractedWith)
            if. Interactables[1].SucceededEvent.Subscribe(OnGameStartButtonBlueInteractedWith)


    OnEndSimulation<override>():void=
        for:
            Player : Entity.GetPlayspaceForEntity[].GetPlayers()
        do:
            RemoveUI(Player)


    # Set red player.
    OnGameStartButtonRedInteractedWith(Agent:agent):void=
        Print("Red Player selected")
        set MaybeRedPlayer = option. Agent
        set RedPlayerReady = true
        CheckRaceStart()


    # Set blue player.
    OnGameStartButtonBlueInteractedWith(Agent:agent):void=
        Print("Blue Player selected")
        set MaybeBluePlayer = option. Agent
        set BluePlayerReady = true
        CheckRaceStart()


    # Check if the race is ready to start.
    CheckRaceStart():void=
        # Check if both players are ready.
        if (RedPlayerReady? and BluePlayerReady?):
            if:
                RedSpawnTransform := MaybeRedSpawnPosition?.GetGlobalTransform()
                BlueSpawnTransform := MaybeBlueSpawnPosition?.GetGlobalTransform()
                BluePlayer := MaybeBluePlayer?
                RedPlayer := MaybeRedPlayer?
            then:
                # Teleport players to their starting positions.
                var Transform: (/Verse.org/SpatialMath:)transform = (/Verse.org/SpatialMath:)transform{}
                if(RedFort := RedPlayer.GetFortCharacter[], BlueFort := BluePlayer.GetFortCharacter[]):
                    if. BlueFort.TeleportTo[
                        (/UnrealEngine.com/Temporary/SpatialMath:)vector3{
                            X := BlueSpawnTransform.Translation.Forward, 
                            Y := -BlueSpawnTransform.Translation.Left, 
                            Z := BlueSpawnTransform.Translation.Up + 300.0}, 
                        (/UnrealEngine.com/Temporary/SpatialMath:)rotation{}]

                    if. RedFort.TeleportTo[
                        (/UnrealEngine.com/Temporary/SpatialMath:)vector3{
                            X := RedSpawnTransform.Translation.Forward, 
                            Y := -RedSpawnTransform.Translation.Left, 
                            Z := RedSpawnTransform.Translation.Up + 300.0}, 
                        (/UnrealEngine.com/Temporary/SpatialMath:)rotation{}]
                
                    then:
                        set RaceStarted = true
                        spawn. StartCountdown(RedFort, BlueFort)

            else Print("Error: Spawn positions not setup on the component.")


    # Block the players on the starting position and start the countdown.
    StartCountdown(RedFort:fort_character, BlueFort:fort_character)<suspends>:void=
        if(RaceStarted?):
            RedFort.PutInStasis(stasis_args{AllowTurning := true})
            BlueFort.PutInStasis(stasis_args{AllowTurning := true})
            if. BluePlayer := player[MaybeBluePlayer?] then. InitUI(BluePlayer)
            if. RedPlayer := player[MaybeRedPlayer?] then. InitUI(RedPlayer)
            if. var counter:int = Int[Countdown]
            then. CountdownTick(counter, RedFort, BlueFort)
        
        else return

    # Countdown function, decrease a value and when ends the race starts.
    CountdownTick(TimeLeft:int, RedFort:fort_character, BlueFort:fort_character)<suspends>:void=
        if (RaceStarted?):
            if (TimeLeft > 0):
                if (BlueUI := PlayersUI[BlueFort.GetAgent[]], RedUI := PlayersUI[RedFort.GetAgent[]]):
                    BlueUI.UpdateTimerUI(TimeLeft)
                    RedUI.UpdateTimerUI(TimeLeft)

                Sleep(1.0)
                CountdownTick(TimeLeft - 1, RedFort, BlueFort)
            else:
                if (BlueUI := PlayersUI[BlueFort.GetAgent[]], RedUI := PlayersUI[RedFort.GetAgent[]]):
                    BlueUI.UpdateTimerUI(TimeLeft)
                    RedUI.UpdateTimerUI(TimeLeft)

                RedFort.ReleaseFromStasis()
                BlueFort.ReleaseFromStasis()
                SpawnBombs()

                # Add timers UI the players.
                if:
                    Sim := Entity.GetSimulationEntity[]
                    RaceManagers := Sim.FindDescendantEntitiesWithComponent(race_manager_component)
                    UIManagers := Sim.FindDescendantEntitiesWithComponent(ui_manager_component)
                    BluePlayer := MaybeBluePlayer?
                    RedPlayer := MaybeRedPlayer?
                    BlueBomb := AgentToBombs[BluePlayer]
                    RedBomb := AgentToBombs[RedPlayer]
                then:
                    for(RaceManager : RaceManagers, Comp := RaceManager.GetComponent[race_manager_component]):
                        Comp.StartRace(BluePlayer, RedPlayer)
                    for(UIManager : UIManagers, Comp := UIManager.GetComponent[ui_manager_component]):
                        Comp.Init(BlueBomb, RedBomb)

                # Remove Countdown UI from players.
                Sleep(1.0)
                if. BluePlayer := player[MaybeBluePlayer?] then. RemoveUI(BluePlayer)
                if. RedPlayer := player[MaybeRedPlayer?] then. RemoveUI(RedPlayer)

        else return


    # Spawn a bomb for each player.
    SpawnBombs<private>():void=
        if (SimEntity := Entity.GetSimulationEntity[]):
            for (Player : Entity.GetPlayspaceForEntity[].GetPlayers()):
                AddBombsTo(Player)


    # Assign a bomb to a player
    AddBombsTo(Agent:agent):void=
        if (SimEntity := Entity.GetSimulationEntity[]):
            Bomb := Prefabs.bomb{}
            SimEntity.AddEntities(array{Bomb})

            if:
                BombComp := Bomb.GetComponent[bomb_component]
                set AgentToBombs[Agent] = BombComp
            then:
                BombComp.Init(Agent)

            else Print("Failed to get bomb component", ?Color := NamedColors.Red)


    # Set logics false when race ends.
    RaceEnded<public>():void=
        set RaceStarted = false
        set RedPlayerReady = false
        set BluePlayerReady = false


# Subclassing interactable_component to exactly find this component in the scene.
race_start_component<public> := class(interactable_component):

# Messages for countdown UI.
MessageTextInt<localizes>(TimeLeft:int):message= "{TimeLeft}"
MessageTextChar<localizes>:message= "START"

timer_ui := class():
    var CountdownTextBlock: text_block = text_block:
        DefaultJustification := text_justification.Center
    var Canvas<internal>: canvas = canvas{}

    block:
        set Canvas = canvas:
            Slots := array{
                canvas_slot:
                    Widget := CountdownTextBlock
                    Anchors := anchors{Minimum:=vector2{X:=0.5, Y:=0.5}, Maximum:=vector2{X:=0.5, Y:=0.5}}
                    Offsets := margin{Left:=0.0, Top:=-350.0}
            }   

    UpdateTimerUI<public>(TimeLeft:int):void=
        if (TimeLeft > 0):
            CountdownTextBlock.SetText(MessageTextInt(TimeLeft))
        else:
            CountdownTextBlock.SetText(MessageTextChar)