<#
@UUID: b2263ff4-3e0d-43d5-8ec4-52f71ffb22b0
@Task: Make a component that points towards a target.
@Description: Implement a SceneGraph component that inditate to the player where its next target is.
@Tags: Gameplay, Rotation
@Export: true
#>

using { /Fortnite.com/Characters }
using { /UnrealEngine.com }
using { /Verse.org }
using { /Verse.org/SceneGraph }
using { /Verse.org/Simulation }
using { /Verse.org/SpatialMath }


# Class for the logic of the target_arrow entity.
target_arrow_component<public> := class<final_super>(component):
    var MaybeAgent: ?agent = false
    var MaybeLookAtTransform: ?transform = false
    var ShouldBeUpdating: logic = false


    OnBeginSimulation<override>():void =
        (super:)OnBeginSimulation()
        set ShouldBeUpdating = true
        spawn. UpdateRotation()
        

    OnEndSimulation<override>():void=
        (super:)OnEndSimulation()
        Entity.RemoveFromParent()


    # Destroy the arrow.
    DestroyArrow<public>():void=
        set ShouldBeUpdating = false
        Entity.RemoveFromParent()


    # Set the transform the arrow should be pointing to.
    SetLookAtTransform<public>(InTransform : transform):void=
        set MaybeLookAtTransform = option. InTransform


    Init<public>(Agent:agent):void=
        set MaybeAgent = option. Agent
        if (Fort := MaybeAgent?.GetFortCharacter[]):
            RotateTowardsTarget()
            

    RotateTowardsTarget():void=
        if (LookAtTransform := MaybeLookAtTransform?, Fort := MaybeAgent?.GetFortCharacter[]):
            LookDirection := (LookAtTransform.Translation - Entity.GetGlobalTransform().Translation).MakeUnitVector()
            LookAtYaw := RadiansToDegrees(ArcTan(-LookDirection.Left,LookDirection.Forward))
            LookAtPitch := RadiansToDegrees(ArcTan(LookDirection.Up,Sqrt((LookDirection.Forward * LookDirection.Forward) + (LookDirection.Left * LookDirection.Left))))
            LookAtRoll := 0.0
            LookAtRotation := MakeRotationFromYawPitchRollDegrees(LookAtYaw, LookAtPitch - 90.0, LookAtRoll)
            Entity.SetGlobalTransform((/Verse.org/SpatialMath:)transform{
                Translation := vector3{Forward := Fort.GetTransform().Translation.X, Left := -Fort.GetTransform().Translation.Y, Up := Fort.GetTransform().Translation.Z + 100.0},
                Rotation := LookAtRotation,
                Scale := Entity.GetGlobalTransform().Scale
                })


    UpdateRotation()<suspends>:void=
        loop:
            Sleep(0.0)
            if (MaybeLookAtTransform?):
                RotateTowardsTarget()

            if (not ShouldBeUpdating?):
                break