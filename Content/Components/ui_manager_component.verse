<#
@UUID: 2a537717-664f-4b58-a07c-48023e5613df
@Task: A component to manage the UI of both players.
@Description: Implement a SceneGraph component that will manage the UI of bith players.
@Tags: UI
@Export: true
#>

using { /Fortnite.com/UI }
using { /UnrealEngine.com/Temporary/SpatialMath }
using { /UnrealEngine.com/Temporary/UI }
using { /Verse.org/SceneGraph }
using { /Verse.org/Simulation }
using { Prefabs }

# Class for the logic of the ui_manager entity.
ui_manager_component<public> := class<final_super>(component):

    var MaybeBlueBomb: ?bomb_component = false
    var MaybeRedBomb: ?bomb_component = false
    var MaybeRedPlayer: ?agent = false
    var MaybeBluePlayer: ?agent = false
    var PlayersUI: [agent]agent_ui = map{}


    OnEndSimulation<override>():void=
        if:
            BluePlayer := MaybeBluePlayer?
            RedPlayer := MaybeRedPlayer?
            BluePlayerCanvas := PlayersUI[BluePlayer]
            RedPlayerCanvas := PlayersUI[RedPlayer]
            BluePlayerUI := GetPlayerUI[player[BluePlayer]]
            RedPlayerUI := GetPlayerUI[player[RedPlayer]]
        then:
            BluePlayerUI.RemoveWidget(BluePlayerCanvas.Canvas)
            RedPlayerUI.RemoveWidget(RedPlayerCanvas.Canvas)
    

    Init(BlueBomb:bomb_component, RedBomb:bomb_component):void=
        if (Sim := Entity.GetSimulationEntity[]):
            Bombs := Sim.FindDescendantEntitiesWithComponent(bomb_component)
            var Temp :[]entity = array{}
            for(Bomb : Bombs):
                set Temp += array. Bomb
            if (BluePlayer := bomb[BlueBomb.Entity].GetAgent[],RedPlayer := bomb[RedBomb.Entity].GetAgent[]):
                set MaybeBlueBomb = option. BlueBomb
                set MaybeRedBomb = option. RedBomb
                set MaybeBluePlayer = option. BluePlayer
                set MaybeRedPlayer = option. RedPlayer
                if. set PlayersUI[RedPlayer] = agent_ui{}
                if. set PlayersUI[BluePlayer] = agent_ui{}

        # Generating and adding UI to both players.
        GenerateUI()
        spawn. TextUpdating()


    # Remove UI from Agent.
    RemoveUI<public>(Agent:agent):void=
        if(AgentUI := PlayersUI[Agent], PlayerUI := GetPlayerUI[player[Agent]]):
            PlayerUI.RemoveWidget(AgentUI.Canvas)


    # Generate and add UI to both Red and Blue players.
    GenerateUI():void=
        UpdateTexts()
        if(BlueBomb := MaybeBlueBomb?, RedBomb := MaybeRedBomb?, BluePlayer := MaybeBluePlayer?, RedPlayer := MaybeRedPlayer?):
            if(BluePlayerUI := GetPlayerUI[player[BluePlayer]], RedPlayerUI := GetPlayerUI[player[RedPlayer]]):
                if(BlueUI := PlayersUI[BluePlayer], RedUI := PlayersUI[RedPlayer]):
                    RedPlayerUI.AddWidget(RedUI.Canvas)
                    BluePlayerUI.AddWidget(BlueUI.Canvas)
    
    
    # Update the text on screen on loop.
    TextUpdating()<suspends>:void=
        loop:
            Sleep(0.1)
            UpdateTexts()
            if (not MaybeBlueBomb?.IsPlayerAlive or not MaybeRedBomb?.IsPlayerAlive):
                break


    # Update the text for both players.
    UpdateTexts<public>():void=
        if (RedBomb := MaybeRedBomb?, BlueBomb := MaybeBlueBomb?, RedAgent := bomb[RedBomb.Entity].GetAgent[], BlueAgent := bomb[BlueBomb.Entity].GetAgent[]):
            if (BlueUI := PlayersUI[BlueAgent], RedUI := PlayersUI[RedAgent]):
                BlueUI.UpdateUI(RedAgent,RedBomb.PlayerTimer,BlueAgent, BlueBomb.PlayerTimer)
                RedUI.UpdateUI(RedAgent,RedBomb.PlayerTimer,BlueAgent, BlueBomb.PlayerTimer)


PlayerMessage<localizes>(Agent:agent,Time:int):message="{Agent}: {Time}"


agent_ui := class():
    RedPlayerTextBlock: text_block = text_block{}
    BluePlayerTextBlock: text_block = text_block{}
    var Canvas<internal>: canvas = canvas{}
    block:
        set Canvas = canvas:
            Slots := array{
                canvas_slot{
                    Widget := RedPlayerTextBlock
                    Anchors := anchors{Minimum := vector2{X := 0.4, Y := 0.0}, Maximum := vector2{X := 0.4, Y := 0.0}}
                    Offsets := margin{Top := 75.0}
                    SizeToContent := true
                },
                canvas_slot{
                    Widget := BluePlayerTextBlock
                    Anchors := anchors{Minimum := vector2{X := 0.4, Y := 0.0}, Maximum := vector2{X := 0.4 , Y := 0.0}}
                    SizeToContent := true
                }
            }

    UpdateUI<public>(RedAgent:agent, RedTime:int,BlueAgent:agent, BlueTime:int):void=
        RedPlayerTextBlock.SetText(PlayerMessage(RedAgent, RedTime))
        BluePlayerTextBlock.SetText(PlayerMessage(BlueAgent, BlueTime))