<#
@UUID: a5f02820-4d7c-4608-ac50-d4156ad78373
@Task: Editor Entites Prefab Setup
@Description: A file containing detailed information about how to create and setup the entity prefabs needed for the game. 
@Tags: Editor, Setup.
@Export: true
#>

<#
In Editor bomb Entity Prefab Setup:
- bomb
    - transform_component
    - mesh_component
        - Collidable = false
    -bomb_component
        - PlayerTimer = 15
        - AddTimerValue = 5
        - BombDamage = 200.0
#>

<#
In Editor character_collision_proxy Entity Prefab Setup:
- character_collision_proxy
    - transform_component
    - mesh_component
        - Visible = false
    - character_collision_proxy_component
#>

<#
In Editor collision_proxy_spawner Entity Prefab Setup:
- collision_proxy_spawner
    - collision_proxy_spawner_component
    - transform_component
#>

<#
In Editor death_volume Entity Prefab Setup:
- death_volume
    - transform_component
    - mesh_component
        - CollisionPreset = OverlapAllDynamic
    - death_volume_component
#>

<#
In Editor jump_pad Entity Prefab Setup:
- jump_pad
    - transform_component
    - mesh_component
    - jump_pad_component
        - BouncerDevice must be set in the map on each jump_pad, drag a bouncer_device in the map and customize it.
#>

<#
In Editor race_manager Entity Prefab Setup:
- race_manager
    - transform_component
    - race_manager_component
        - RaceCheckpoints must be filled in the map with checkpoints following their CheckPointIndex order.
#>

<#
In Editor start_race Entity Prefab Setup:
- start_race
    - transform_component
    - start_game_component
        - MaybeRedSpawnPosition = red_starting_plate
        - MaybeBlueSpawnPosition = blue_starting_plate
    - blue_starting_plate (entity)
        - mesh_component
        - transform_component
    - blue_interactable (entity)
        - transform_component
        - mesh_component
        - race_start_component
    - red_starting_plate (entity)
        - mesh_component
        - transform_component
    - red_interactable (entity)
        - transform_component
        - mesh_component
        - race_start_component

#>

<#
In Editor checkpoint Entity Prefab Setup:
- checkpoint
    - transform_component
    - mesh_component
        - Collidable = false
    -checkpoint_component
        - MaybeBluePlayerLight = p1_active_light
        - aybeRedPlayerLight = p2_active_light
        - CheckpointIndex has to be set in the editor with the correct order.    
        - p1_active_light (entity)
            - transform_component
            - spot_light_component
                - any spot_light_component setting
        - p2_active_light (entity)
            - transform_component
            - spot_light_component
                - any spot_light_component setting
#>

<#
In Editor target_arrow Entity Prefab Setup:
- target_arrow
    - transform_component
    - mesh_component
        - Collidable = false
    -target_arrow_component
#>

<#
In Editor ui_manager Entity Prefab Setup:
- ui_manager
    - transform_component
    -ui_manager_component
#>

