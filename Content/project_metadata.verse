<#
@UUID: 3e2ce1d2-ed7e-4ca2-96ab-6da169ae418f
@Task: Bomb Run.
@Description: A minigame where two players race across multiple checkpoints avoiding obstacles. Each player has a ticking bomb attached to it whose timer gets incremented when the player reaches a checkpoint. The player that reaches the end of the race alive first, wins the game.
@Tags: Minigame.
@Export: true
#>