<#
@UUID: 98665766-65ca-4e0a-9b88-df4857c8d5be
@Task: Make a component to manage a timed race.
@Description: Implement a SceneGraph component that manages a timed player-versus-player race with checkpoints that increase a countdown timer.
@Tags: Checkpoint, Race Manager
@Export: true
#>

using { /Fortnite.com/Devices }
using { /Verse.org }
using { /Verse.org/Native }
using { /Verse.org/SceneGraph}
using { /Verse.org/Simulation}
using { /Verse.org/SpatialMath}
using { Components }
using { Prefabs }


# Class for the logic of the race_manager entity.
race_manager_component<public> := class<final_super>(component):
    @editable
    MessageDevice: hud_message_device = hud_message_device{}
    @editable
    RaceCheckpoints: []checkpoint = array{}

    var AgentToCheckpointIndexMap: [agent]?int = map{}
    var AgentToArrow: [agent]entity = map{}
    var MaybeRedPlayer: ?agent = false
    var MaybeBluePlayer: ?agent = false


    # === Entity Functions ===


    OnBeginSimulation<override>():void=
        (super:)OnBeginSimulation()

        SubscribeToCheckpointEvents()


    # === Race Manager Functions ===


    # Start the race.
    StartRace<public>(BluePlayer:agent, RedPlayer:agent):void=
        InitializePlayers(BluePlayer, RedPlayer)

        if:
            FirstCheckpoint := RaceCheckpoints[0]
            Comp := FirstCheckpoint.GetComponent[checkpoint_component]
            RedArrow := AgentToArrow[RedPlayer].GetComponent[target_arrow_component]
            BlueArrow := AgentToArrow[BluePlayer].GetComponent[target_arrow_component]
        then:
            Comp.EnableCheckpointForPlayer(BluePlayer, true)
            Comp.EnableCheckpointForPlayer(RedPlayer, false)

            BlueArrow.SetLookAtTransform(FirstCheckpoint.GetGlobalTransform())
            RedArrow.SetLookAtTransform(FirstCheckpoint.GetGlobalTransform())


    # Initialize agent variables.
    InitializePlayers<public>(BluePlayer:agent, RedPlayer:agent):void=
        if:
            Sim := Entity.GetSimulationEntity[]
            set MaybeRedPlayer = option. RedPlayer
            set MaybeBluePlayer = option. BluePlayer
            set AgentToCheckpointIndexMap[BluePlayer] = option{0}
            set AgentToCheckpointIndexMap[RedPlayer] = option{0}
            BlueArrow := target_arrow{}
            BlueArrowComp := BlueArrow.GetComponent[target_arrow_component]
            RedArrow := target_arrow{}
            RedArrowComp := RedArrow.GetComponent[target_arrow_component]
        then:
            Sim.AddEntities(array{BlueArrow})
            BlueArrowComp.Init(BluePlayer)
            if. set AgentToArrow[BluePlayer] = BlueArrow

            Sim.AddEntities(array{RedArrow})
            RedArrowComp.Init(RedPlayer)
            if. set AgentToArrow[RedPlayer] = RedArrow


    # Activate next checkpoint.
    ActivateNextCheckpoint(InAgent:agent):void=
        var CurrentCheckpointIndex: int = 0
        var IsBluePlayer: logic = false
        if. InAgent = MaybeBluePlayer? then. set IsBluePlayer = true else. set IsBluePlayer = false
        if (Index := AgentToCheckpointIndexMap[InAgent]?):
            set CurrentCheckpointIndex = Index

        # Check if the race is over.
        if (CurrentCheckpointIndex = RaceCheckpoints.Length - 1):
            if:
                CurrentCheckpoint := RaceCheckpoints[CurrentCheckpointIndex]
                CurrentCheckpointComponent := CurrentCheckpoint.GetComponent[Components.checkpoint_component]
            then:
                CurrentCheckpointComponent.DisableCheckpointForPlayer(InAgent, IsBluePlayer)
            DeclareWinner(InAgent) 
            return

        if:
            CurrentCheckpoint := RaceCheckpoints[CurrentCheckpointIndex]
            CurrentCheckpointComponent := CurrentCheckpoint.GetComponent[Components.checkpoint_component]
        then:
            CurrentCheckpointComponent.DisableCheckpointForPlayer(InAgent, IsBluePlayer)

        if:
            NextCheckpoint:checkpoint = RaceCheckpoints[CurrentCheckpointIndex + 1]
            NextCheckpointComponent := NextCheckpoint.GetComponent[Components.checkpoint_component]
            Arrow := AgentToArrow[InAgent].GetComponent[target_arrow_component]
        then:
            Arrow.SetLookAtTransform(NextCheckpoint.GetGlobalTransform())         
            NextCheckpointComponent.EnableCheckpointForPlayer(InAgent, IsBluePlayer)

        # Increment checkpoint index.
        if. set AgentToCheckpointIndexMap[InAgent] = option{CurrentCheckpointIndex + 1}


    # Declare the winner and end the race.
    DeclareWinner(InAgent:agent):void=
        AgentName: message = GetAgentName(InAgent)
        Message<localizes>: message = " has won the race!"
        WinnerMessage: message = MakeWinnerMessage(AgentName, Message)
        MessageDevice.SetText(WinnerMessage)
        MessageDevice.Show()
        if:
            BluePlayer := MaybeBluePlayer?
            RedPlayer := MaybeRedPlayer?
            Sim := Entity.GetSimulationEntity[]
            Arrows := Sim.FindDescendantEntitiesWithComponent(target_arrow_component)
            Bombs := Sim.FindDescendantEntitiesWithComponent(bomb_component)
            UIManagers := Sim.FindDescendantEntitiesWithComponent(ui_manager_component)
            StartGames := Sim.FindDescendantEntitiesWithComponent(start_game_component)
        then:
            for (Arrow : Arrows, Comp := Arrow.GetComponent[target_arrow_component]):
                Comp.DestroyArrow()
            for (Bomb : Bombs, Comp := Bomb.GetComponent[bomb_component]):
                Comp.RemoveBombsFrom()
            for (UIManager : UIManagers, Comp := UIManager.GetComponent[ui_manager_component]):
                Comp.RemoveUI(BluePlayer)
                Comp.RemoveUI(RedPlayer)
            for (StartGame : StartGames, Comp := StartGame.GetComponent[start_game_component]):
                Comp.RaceEnded()
            ResetRace()
                

    # === Helpers ===


    # Get Checkpoint transform.
    GetCheckpointTransform<private>(CheckpointIndex:int)<decides><transacts>:transform=
        var MaybeTransform: ?transform = false

        if (Checkpoint:checkpoint = RaceCheckpoints[CheckpointIndex]):
            set MaybeTransform = option{Checkpoint.GetGlobalTransform()}

        MaybeTransform?


    # Reset race variables.
    ResetRace<private>():void=
        # Internal variables.
        for (Agent->Index : AgentToCheckpointIndexMap, set AgentToCheckpointIndexMap[Agent] = false):

        # Checkpoints.
        for:
            Checkpoint : RaceCheckpoints
            CheckpointComponent := Checkpoint.GetComponent[checkpoint_component]
        do:
            CheckpointComponent.Reset()


    # Subscribe to checkpoint events.
    SubscribeToCheckpointEvents<private>():void=
        for:
            Index->Checkpoint : RaceCheckpoints
            CheckpointComponent := Checkpoint.GetComponent[checkpoint_component]
        do:
            CheckpointComponent.SetIndex(Index)
            CheckpointComponent.Subscribe(OnCheckpointReached)


    # Handle Checkpoint Reached event.
    OnCheckpointReached<private>(InAgent:agent, InIndex:int):void=
        if (AgentToCheckpointIndexMap[InAgent]? = InIndex):
            if:
                Sim := Entity.GetSimulationEntity[]
                Bombs := Sim.FindDescendantEntities(bomb)          
            then:
                for (Bomb : Bombs):
                    if:
                        InAgent = Bomb.GetAgent[]
                        Comp := Bomb.GetComponent[bomb_component]
                    then:
                        Comp.IncreaseTimer(InAgent)
                        ActivateNextCheckpoint(InAgent)
                        return


    # Create a message with the agent name.
    MakeWinnerMessage<localizes>(M1:message, M2:message):message="{M1}{M2}"
    GetAgentName<localizes>(InAgent:agent):message="{InAgent}"